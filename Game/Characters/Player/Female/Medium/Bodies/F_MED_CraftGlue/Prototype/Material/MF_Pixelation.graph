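# Prototype material function for the CraftGlue player body that renders a translucent shell as a screen-space pixelated mosaic of the scene color behind it.
# A custom HLSL node averages snapped pixel blocks, sized by camera depth fade, with adjustable footprint, subsampling and Gaussian weighting.
# The result feeds Emissive, alongside sorted-translucency opacity and a normal-based vertex offset.
# Earlier edge-aware, median, low-contrast and base-color experiments remain unconnected to the output.
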MaterialFunction /Game/Characters/Player/Female/Medium/Bodies/F_MED_CraftGlue/Prototype/Material/MF_Pixelation
graph {
 properties {description=5/7}
 nodes [
  n1 = ScreenPosition()
  MaterialExpressionViewSize_0 = ViewSize()
  MaterialExpressionDivide_0 = Divide(A=Subtract(A=Multiply(A=@n1.ViewportUV, B=@MaterialExpressionViewSize_0)), B=@MaterialExpressionViewSize_0)
  MaterialExpressionLinearInterpolate_8 = LinearInterpolate(
   A=ScalarParameter() {defaultValue=6, group=Pixelation, parameterName=PixelSize_Close}
   B=ScalarParameter() {defaultValue=3, group=Pixelation, parameterName=PixelSize_Far}
   Alpha=MaterialFunctionCall() {materialFunction=&CameraDepthFade}.Result
  )
  MaterialExpressionSceneColor_0 = SceneColor!(UVs=ScreenPosition().ViewportUV)
  MaterialExpressionScalarParameter_2 = ScalarParameter() {group=Pixelation, parameterName=Threshold}
  Custom!(
   ScreenPosition=@MaterialExpressionDivide_0
   ViewportSize=@MaterialExpressionViewSize_0
   BlockSize=@MaterialExpressionLinearInterpolate_8
   SceneTexture=@MaterialExpressionSceneColor_0.RGB
   Threshold=@MaterialExpressionScalarParameter_2
  ) {
   code="// Inputs:\r\n// float BlockSize\r\n// float Threshold\r\n\r\nfloat2 viewportUV = ScreenPosition.xy;\r\n\r\n// Convert to scene UVs\r\nfloat2 sceneUV =\r\n    viewportUV *\r\n    View.ViewSizeAndInvSize.xy *\r\n    View.BufferSizeAndInvSize.zw;\r\n\r\n// Buffer size (in pixels)\r\nfloat2 bufferSize = View.BufferSizeAndInvSize.xy;\r\n\r\n// Convert UV → pixel coordinates\r\nfloat2 pixelCoord = floor(sceneUV * bufferSize);\r\n\r\n// Ensure stable integer block size\r\nfloat blockSize = max(1, floor(BlockSize));\r\n\r\n// Find the start of the pixel block\r\nfloat2 blockStart = floor(pixelCoord / blockSize) * blockSize;\r\n\r\n// Sample center color (reference)\r\nfloat2 centerUV = (blockStart + blockSize * 0.5) / bufferSize;\r\nfloat3 centerCol = DecodeSceneColorAndAlpharForMaterialNode(centerUV).rgb;\r\n\r\n// Accumulation\r\nfloat3 accum = 0;\r\nint count = 0;\r\n\r\n// Loop through block\r\n[loop]\r\nfor (int y = 0; y < blockSize; y++)\r\n{\r\n    [loop]\r\n    for (int x = 0; x < blockSize; x++)\r\n    {\r\n        float2 sampleCoord = blockStart + float2(x, y);\r\n        float2 sampleUV = (sampleCoord + 0.5) / bufferSize;\r\n\r\n        float3 col = DecodeSceneColorAndAlpharForMaterialNode(sampleUV).rgb;\r\n\r\n        // Only accumulate similar colors (edge-aware)\r\n        if (distance(col, centerCol) < Threshold)\r\n        {\r\n            accum += col;\r\n            count++;\r\n        }\r\n    }\r\n}\r\n\r\n// Fallback to center color if nothing matched\r\nfloat3 finalCol = (count > 0) ? (accum / count) : centerCol;\r\n\r\nreturn float4(finalCol, 1);"
   description=High-contrast version
   inputs [
    {
     inputName=ScreenPosition
     input {expression=&self:MaterialExpressionDivide_0}
    }
    {
     inputName=ViewportSize
     input {expression=&self:MaterialExpressionViewSize_0}
    }
    {
     inputName=BlockSize
     input {expression=&self:MaterialExpressionLinearInterpolate_8}
    }
    {
     inputName=SceneTexture
     input {
      expression=&self:MaterialExpressionSceneColor_0
      mask=1
      maskB=1
      maskG=1
      maskR=1
     }
    }
    {
     inputName=Threshold
     input {expression=&self:MaterialExpressionScalarParameter_2}
    }
   ]
  }
  Custom(
   ScreenPosition=@MaterialExpressionDivide_0
   ViewportSize=@MaterialExpressionViewSize_0
   BlockSize=@MaterialExpressionLinearInterpolate_8
   SceneTexture=@MaterialExpressionSceneColor_0.RGB
   Threshold=@MaterialExpressionScalarParameter_2
  ) {
   code="// Inputs:\r\n//   float  BlockSize   - desired pixel-block edge length (px)\r\n//   float  Threshold   - UNUSED in these modes (kept for graph/pin compatibility)\r\n//\r\n// Pixelization tuned for jittering scene color on flat-color / sharp-edge assets.\r\n// No edge-aware reference sample anymore. Pick an aggregation with MODE:\r\n//\r\n//   MODE 0  Gaussian-weighted average (no reference). Smooth, but averages\r\n//           across edges -> can still shimmer where colors blend.\r\n//   MODE 1  Per-channel median (brute force). Robust to outliers; may produce\r\n//           a color not actually present (channels chosen independently).\r\n//   MODE 2  Luminance median (brute force). Returns an ACTUAL sampled color\r\n//           (the block's median-luminance texel). Best for flat/sharp content:\r\n//           picks the dominant color, ignores jittering outliers, no blending.\r\n//\r\n// Median modes are O(count^2) ALU + store the whole block, so block size is\r\n// clamped to MAX_BLOCK to bound the sample buffer.\r\n##define MODE          0\r\n##define GAUSS_FALLOFF 2.5   // MODE 0 only: higher = tighter gaussian\r\n##define MAX_BLOCK     8     // median buffer is MAX_BLOCK*MAX_BLOCK samples\r\n\r\nconst float3 LUMA = float3(0.299, 0.587, 0.114);\r\n\r\n// ---- setup ---------------------------------------------------------------\r\nfloat2 viewportUV = ScreenPosition.xy;\r\n\r\n// Convert to scene UVs\r\nfloat2 sceneUV =\r\n    viewportUV *\r\n    View.ViewSizeAndInvSize.xy *\r\n    View.BufferSizeAndInvSize.zw;\r\n\r\nfloat2 bufferSize = View.BufferSizeAndInvSize.xy;   // buffer size (px)\r\nfloat2 pixelCoord = floor(sceneUV * bufferSize);    // UV -> pixel coords\r\n\r\n// Stable integer block (clamped so the median buffer stays bounded), snapped grid\r\nfloat  blockSize   = min(max(1, floor(BlockSize)), (float)MAX_BLOCK);\r\nfloat2 blockStart  = floor(pixelCoord / blockSize) * blockSize;\r\nfloat2 blockCenter = blockStart + blockSize * 0.5;\r\nfloat  halfBlock   = max(blockSize * 0.5, 1e-4);\r\n\r\nfloat3 finalCol;\r\n\r\n##if MODE == 0\r\n// ==== Gaussian-weighted average, no reference =============================\r\nfloat3 accum = 0;\r\nfloat  wsum  = 0;\r\n\r\n[loop]\r\nfor (int y = 0; y < blockSize; y++)\r\n{\r\n    [loop]\r\n    for (int x = 0; x < blockSize; x++)\r\n    {\r\n        float2 sampleCoord = blockStart + float2(x, y) + 0.5;\r\n        float3 col = DecodeSceneColorAndAlpharForMaterialNode(sampleCoord / bufferSize).rgb;\r\n\r\n        float2 n = (sampleCoord - blockCenter) / halfBlock;   // ~[-1,1]\r\n        float  w = exp(-GAUSS_FALLOFF * dot(n, n));\r\n\r\n        accum += col * w;\r\n        wsum  += w;\r\n    }\r\n}\r\nfinalCol = accum / max(wsum, 1e-4);\r\n\r\n##else\r\n// ==== Brute-force median =================================================\r\n// Gather every texel in the block into a fixed buffer.\r\nfloat3 cols[MAX_BLOCK * MAX_BLOCK];\r\nint count = 0;\r\n\r\n[loop]\r\nfor (int gy = 0; gy < blockSize; gy++)\r\n{\r\n    [loop]\r\n    for (int gx = 0; gx < blockSize; gx++)\r\n    {\r\n        float2 sampleCoord = blockStart + float2(gx, gy) + 0.5;\r\n        cols[count] = DecodeSceneColorAndAlpharForMaterialNode(sampleCoord / bufferSize).rgb;\r\n        count++;\r\n    }\r\n}\r\n\r\nint target = count / 2;   // middle rank (upper-median for even counts)\r\n\r\n##if MODE == 1\r\n// ---- per-channel median -------------------------------------------------\r\n// For each channel, find the value whose rank == target. Tie-break by sample\r\n// index so identical flat colors get a strict order (otherwise all tie at 0).\r\nfloat3 med = cols[0];\r\n[loop]\r\nfor (int c = 0; c < 3; c++)\r\n{\r\n    [loop]\r\n    for (int i = 0; i < count; i++)\r\n    {\r\n        float vi = cols[i][c];\r\n        int rank = 0;\r\n        [loop]\r\n        for (int j = 0; j < count; j++)\r\n        {\r\n            float vj = cols[j][c];\r\n            if (vj < vi || (vj == vi &&&& j < i)) rank++;\r\n        }\r\n        if (rank == target) med[c] = vi;\r\n    }\r\n}\r\nfinalCol = med;\r\n\r\n##else\r\n// ---- luminance median (returns an actual sampled color) -----------------\r\nfloat3 medianCol = cols[0];\r\n[loop]\r\nfor (int i = 0; i < count; i++)\r\n{\r\n    float li = dot(cols[i], LUMA);\r\n    int rank = 0;\r\n    [loop]\r\n    for (int j = 0; j < count; j++)\r\n    {\r\n        float lj = dot(cols[j], LUMA);\r\n        if (lj < li || (lj == li &&&& j < i)) rank++;\r\n    }\r\n    if (rank == target) medianCol = cols[i];\r\n}\r\nfinalCol = medianCol;\r\n##endif\r\n\r\n##endif\r\n\r\nreturn float4(finalCol, 1);\r\n"
   description=REDO 2: High-contrast version
   inputs [
    {
     inputName=ScreenPosition
     input {expression=&self:MaterialExpressionDivide_0}
    }
    {
     inputName=ViewportSize
     input {expression=&self:MaterialExpressionViewSize_0}
    }
    {
     inputName=BlockSize
     input {expression=&self:MaterialExpressionLinearInterpolate_8}
    }
    {
     inputName=SceneTexture
     input {
      expression=&self:MaterialExpressionSceneColor_0
      mask=1
      maskB=1
      maskG=1
      maskR=1
     }
    }
    {
     inputName=Threshold
     input {expression=&self:MaterialExpressionScalarParameter_2}
    }
   ]
  }
  n2 = Constant() {r=1}
  n3 = VertexNormalWS()
  n4 = ScalarParameter() {defaultValue=0.1, group=Pixelation, parameterName=Vertex Offset}
  FunctionOutput(None=SetMaterialAttributes!(
    MaterialAttributes=FunctionInput() {inputName=MaterialAttributes, inputType=FunctionInput_MaterialAttributes}
    Opacity=MaterialFunctionCall!(Opacity=@n2) {materialFunction=/Game/Global/Materials/MaterialGym/Materials/SortedOutline/MF_SortTranslucency}.Result
    WorldPositionOffset=Multiply(A=@n3, B=@n4)
    EmissiveColor=NamedRerouteUsage()
   ) {attributeSetTypes [#1, #2, #3]})
  n5 = ScreenPosition()
  MaterialExpressionViewSize_2 = ViewSize()
  MaterialExpressionDivide_3 = Divide(
   A=Subtract(A=Multiply(A=@n5.ViewportUV, B=@MaterialExpressionViewSize_2), B=ViewProperty() {property=MEVP_TemporalSampleOffset}.Property)
   B=@MaterialExpressionViewSize_2
  )
  MaterialExpressionLinearInterpolate_0 = LinearInterpolate(
   A=LinearInterpolate(
    A=ScalarParameter() {defaultValue=6, group=Pixelation, parameterName=Pixelation Amount - Close}
    B=ScalarParameter() {defaultValue=6, group=Pixelation, parameterName=High - Close}
    Alpha=ScalarParameter() {group=Pixelation, parameterName=High Contrast}
   ) {constB=6}
   B=LinearInterpolate(
    A=ScalarParameter() {defaultValue=3, group=Pixelation, parameterName=Pixelation Amount - Far}
    B=ScalarParameter() {defaultValue=3, group=Pixelation, parameterName=High - Far}
    Alpha=ScalarParameter() {group=Pixelation, parameterName=High Contrast}
   ) {constB=3}
   Alpha=MaterialFunctionCall() {materialFunction=&CameraDepthFade}.Result
  )
  MaterialExpressionSceneColor_1 = SceneColor()
  n6 = Custom!(
   ScreenPosition=@MaterialExpressionDivide_3
   ViewportSize=@MaterialExpressionViewSize_2
   BlockSize=@MaterialExpressionLinearInterpolate_0
   SceneTexture=@MaterialExpressionSceneColor_1.RGB
  ) {
   code="float2 viewportUV = ScreenPosition.xy;\r\n\r\n\r\nfloat2 sceneUV =\r\n    viewportUV *\r\n    View.ViewSizeAndInvSize.xy *\r\n    View.BufferSizeAndInvSize.zw;\r\n\r\nfloat2 bufferSize = View.BufferSizeAndInvSize.xy;\r\n\r\nfloat2 pixelCoord = floor(sceneUV * bufferSize);\r\n\r\nfloat2 blockStart = floor(pixelCoord / BlockSize) * BlockSize;\r\n\r\nfloat3 accum = 0;\r\nint count = 0;\r\n\r\n\r\n\r\n[loop]\r\nfor (int y = 0; y < BlockSize; y++)\r\n{\r\n    [loop]\r\n    for (int x = 0; x < BlockSize; x++)\r\n    {\r\n        \r\n        float2 sampleCoord = blockStart + float2(x, y);\r\n        float2 sampleUV = (sampleCoord + 0.5) / bufferSize;\r\n        \r\n\r\n        float3 col = DecodeSceneColorAndAlpharForMaterialNode(sampleUV).rgb;\r\n\r\n        accum += col;\r\n        count++;\r\n    }\r\n}\r\n\r\nreturn float4(accum / count, 1);"
   description=Low-contrast Version
   inputs [
    {
     inputName=ScreenPosition
     input {expression=&self:MaterialExpressionDivide_3}
    }
    {
     inputName=ViewportSize
     input {expression=&self:MaterialExpressionViewSize_2}
    }
    {
     inputName=BlockSize
     input {expression=&self:MaterialExpressionLinearInterpolate_0}
    }
    {
     inputName=SceneTexture
     input {
      expression=&self:MaterialExpressionSceneColor_1
      mask=1
      maskB=1
      maskG=1
      maskR=1
     }
    }
   ]
  }
  n7 = Custom!(
   ScreenPosition=@MaterialExpressionDivide_0
   ViewportSize=@MaterialExpressionViewSize_0
   BlockSize=@MaterialExpressionLinearInterpolate_8
   SceneTexture=@MaterialExpressionSceneColor_0.RGB
   Threshold=@MaterialExpressionScalarParameter_2
  ) {
   code="// Inputs:\r\n//   float  BlockSize   - desired pixel-block edge length (px)\r\n//   float  Threshold   - color falloff width: 0 = sharp edges, larger = softer / more blending\r\n//\r\n// Temporally-stable, edge-aware pixelization. Designed for jittering scene color\r\n// (e.g. residual TSR/temporal instability in the sampled buffer).\r\n//\r\n// Two changes vs. the original to stop the flicker:\r\n//   1. NO HARD CUTOFF. Samples are blended with a smooth weight, so a sub-pixel\r\n//      color wobble nudges a sample's contribution by an epsilon instead of\r\n//      flipping it fully in/out of the average.\r\n//   2. NO SINGLE CENTER SAMPLE. The edge-aware reference is the block MEAN\r\n//      (computed in pass 1), which is ~N^2x more temporally stable than one tap.\r\n//\r\n// SPATIAL_KERNEL lets you A/B the in-block weighting (recompile to switch):\r\n//   0 = box      (every texel equal -> maximum averaging, most stable, softest mosaic)\r\n//   1 = tent     (radial linear falloff, center-weighted)\r\n//   2 = gaussian (radial, soft center weighting)\r\n##define SPATIAL_KERNEL 2\r\n##define GAUSS_FALLOFF  2.5   // higher = tighter gaussian\r\n\r\n// ---- setup ---------------------------------------------------------------\r\nfloat2 viewportUV = ScreenPosition.xy;\r\n\r\n// Convert to scene UVs\r\nfloat2 sceneUV =\r\n    viewportUV *\r\n    View.ViewSizeAndInvSize.xy *\r\n    View.BufferSizeAndInvSize.zw;\r\n\r\nfloat2 bufferSize = View.BufferSizeAndInvSize.xy;   // buffer size (px)\r\nfloat2 pixelCoord = floor(sceneUV * bufferSize);    // UV -> pixel coords\r\n\r\n// Stable integer block, snapped grid\r\nfloat  blockSize   = max(1, floor(BlockSize));\r\nfloat2 blockStart  = floor(pixelCoord / blockSize) * blockSize;\r\nfloat2 blockCenter = blockStart + blockSize * 0.5;\r\nfloat  halfBlock   = max(blockSize * 0.5, 1e-4);\r\n\r\n// ---- pass 1: unweighted block mean (stable edge-aware reference) ----------\r\n// Plain box average on purpose: equal weighting = lowest-variance reference,\r\n// and it decouples the reference from whichever SPATIAL_KERNEL you're testing.\r\nfloat3 meanCol = 0;\r\nfloat  meanN   = 0;\r\n\r\n[loop]\r\nfor (int my = 0; my < blockSize; my++)\r\n{\r\n    [loop]\r\n    for (int mx = 0; mx < blockSize; mx++)\r\n    {\r\n        float2 sampleUV = (blockStart + float2(mx, my) + 0.5) / bufferSize;\r\n        meanCol += DecodeSceneColorAndAlpharForMaterialNode(sampleUV).rgb;\r\n        meanN   += 1.0;\r\n    }\r\n}\r\nmeanCol /= max(meanN, 1.0);\r\n\r\n// ---- pass 2: soft, edge-aware, spatially-weighted average -----------------\r\nfloat3 accum = 0;\r\nfloat  wsum  = 0;\r\n\r\n[loop]\r\nfor (int y = 0; y < blockSize; y++)\r\n{\r\n    [loop]\r\n    for (int x = 0; x < blockSize; x++)\r\n    {\r\n        float2 sampleCoord = blockStart + float2(x, y) + 0.5;\r\n        float2 sampleUV    = sampleCoord / bufferSize;\r\n        float3 col = DecodeSceneColorAndAlpharForMaterialNode(sampleUV).rgb;\r\n\r\n        // Spatial weight from normalized offset to block center (~[-1,1]).\r\n        float2 n = (sampleCoord - blockCenter) / halfBlock;\r\n    ##if   SPATIAL_KERNEL == 0\r\n        float sw = 1.0;                              // box\r\n    ##elif SPATIAL_KERNEL == 1\r\n        float sw = saturate(1.0 - length(n));        // tent\r\n    ##else\r\n        float sw = exp(-GAUSS_FALLOFF * dot(n, n));  // gaussian\r\n    ##endif\r\n\r\n        // Soft edge-aware weight: 1.0 at the mean, fading to 0 at Threshold.\r\n        // Replaces the old binary `if (distance < Threshold)` cutoff.\r\n        float cw = 1.0 - smoothstep(0.0, max(Threshold, 1e-4), distance(col, meanCol));\r\n\r\n        float w = sw * cw;\r\n        accum += col * w;\r\n        wsum  += w;\r\n    }\r\n}\r\n\r\n// Fall back to the (stable) mean if everything was rejected.\r\nfloat3 finalCol = (wsum > 1e-4) ? (accum / wsum) : meanCol;\r\nreturn float4(finalCol, 1);\r\n"
   description=REDO 1: High-contrast version
   inputs [
    {
     inputName=ScreenPosition
     input {expression=&self:MaterialExpressionDivide_0}
    }
    {
     inputName=ViewportSize
     input {expression=&self:MaterialExpressionViewSize_0}
    }
    {
     inputName=BlockSize
     input {expression=&self:MaterialExpressionLinearInterpolate_8}
    }
    {
     inputName=SceneTexture
     input {
      expression=&self:MaterialExpressionSceneColor_0
      mask=1
      maskB=1
      maskG=1
      maskR=1
     }
    }
    {
     inputName=Threshold
     input {expression=&self:MaterialExpressionScalarParameter_2}
    }
   ]
  }
  LinearInterpolate(A=@n6, B=@n7, Alpha=ScalarParameter() {group=Pixelation, parameterName=High Contrast})
  MaterialExpressionScalarParameter_20 = ScalarParameter() {defaultValue=1, group=Pixelation, parameterName=SampleAreaScale}
  MaterialExpressionScalarParameter_21 = ScalarParameter() {group=Pixelation, parameterName=Subsample}
  MaterialExpressionScalarParameter_22 = ScalarParameter() {defaultValue=4, group=Pixelation, parameterName=SamplePerAxis}
  MaterialExpressionScalarParameter_23 = ScalarParameter() {group=Pixelation, parameterName=Weighting}
  MaterialExpressionScalarParameter_24 = ScalarParameter() {defaultValue=2.5, group=Pixelation, parameterName=GaussFalloff}
  n8 = Custom!(
   ScreenPosition=@MaterialExpressionDivide_0
   ViewportSize=@MaterialExpressionViewSize_0
   BlockSize=@MaterialExpressionLinearInterpolate_8
   SceneTexture=@MaterialExpressionSceneColor_0.RGB
   Threshold=@MaterialExpressionScalarParameter_2
   SampleAreaScale=@MaterialExpressionScalarParameter_20
   Subsample=@MaterialExpressionScalarParameter_21
   SamplesPerAxis=@MaterialExpressionScalarParameter_22
   Weighting=@MaterialExpressionScalarParameter_23
   GaussFalloff=@MaterialExpressionScalarParameter_24
  ) {
   code="// Inputs (Custom node pins -> drive from ScalarParameters in a material instance):\r\n//   float  BlockSize        - pixel-block edge length (px)\r\n//   float  Threshold        - UNUSED (kept for pin compatibility)\r\n//   float  SampleAreaScale  - footprint width vs block. 1.0 = block; >1 reaches into\r\n//                             neighbours (more averaging, more stable, softer edges);\r\n//                             <1 tighter (less averaging, sharper, less stable).\r\n//   float  Subsample        - 0 = one tap per texel (full coverage, stability ceiling)\r\n//                             1 = fixed SamplesPerAxis^2 bilinear taps (constant cost)\r\n//   float  SamplesPerAxis   - taps per axis when Subsample >= 0.5 (e.g. 3..6)\r\n//   float  Weighting        - 0 = box (equal weights, max averaging, most stable)\r\n//                             1 = gaussian (center-weighted). Blends between the two.\r\n//   float  GaussFalloff     - gaussian tightness when Weighting > 0 (e.g. 2.5; higher = tighter)\r\n//\r\n// Averaging-based pixelization for jittering scene color. Median popped (discrete);\r\n// averaging wins because it reduces variance. Best baseline so far:\r\n//   SampleAreaScale 1.0, Subsample 0, Weighting 0  (pure box mean of the block).\r\n// Most promising lever: raise SampleAreaScale (1.5-2.0) to add averaging.\r\n\r\nconst float MAX_TAPS_AXIS = 32.0;   // safety cap for the full-coverage loop\r\n\r\n// ---- setup ---------------------------------------------------------------\r\nfloat2 viewportUV = ScreenPosition.xy;\r\n\r\n// Convert to scene UVs\r\nfloat2 sceneUV =\r\n    viewportUV *\r\n    View.ViewSizeAndInvSize.xy *\r\n    View.BufferSizeAndInvSize.zw;\r\n\r\nfloat2 bufferSize = View.BufferSizeAndInvSize.xy;   // buffer size (px)\r\nfloat2 pixelCoord = floor(sceneUV * bufferSize);    // UV -> pixel coords\r\n\r\n// Stable integer block, snapped grid\r\nfloat  blockSize   = max(1, floor(BlockSize));\r\nfloat2 blockStart  = floor(pixelCoord / blockSize) * blockSize;\r\nfloat2 blockCenter = blockStart + blockSize * 0.5;\r\n\r\n// Footprint width (px). Scaling reaches outside the block when > 1.\r\nfloat  scale     = max(SampleAreaScale, 1e-3);\r\nfloat  footprint = blockSize * scale;\r\n\r\n// Tap count per axis (runtime selectable).\r\nfloat samplesPerAxis = max(1.0, round(SamplesPerAxis));\r\nfloat tapsAxisF      = (Subsample >= 0.5) ? samplesPerAxis : round(footprint);\r\nint   N              = (int)clamp(tapsAxisF, 1.0, MAX_TAPS_AXIS);\r\n\r\n// Box vs gaussian blend factor (0 = box, 1 = gaussian).\r\nfloat weighting = saturate(Weighting);\r\n\r\n// ---- accumulate ----------------------------------------------------------\r\nfloat3 accum = 0;\r\nfloat  wsum  = 0;\r\n\r\n[loop]\r\nfor (int iy = 0; iy < N; iy++)\r\n{\r\n    [loop]\r\n    for (int ix = 0; ix < N; ix++)\r\n    {\r\n        // Parametric tap position in [-0.5, 0.5] across the footprint.\r\n        // (At scale 1, Subsample 0, N == blockSize -> exact per-texel centers.)\r\n        float2 p = (float2(ix, iy) + 0.5) / N - 0.5;\r\n        float2 sampleCoord = blockCenter + p * footprint;\r\n\r\n        // Bilinear scene-color fetch (sub-taps average 4 texels for free).\r\n        float3 col = DecodeSceneColorAndAlpharForMaterialNode(sampleCoord / bufferSize).rgb;\r\n\r\n        // Box weight 1.0, gaussian weight exp(...), blended by Weighting.\r\n        float2 n      = p * 2.0;                          // [-1, 1] across footprint\r\n        float  wGauss = exp(-GaussFalloff * dot(n, n));\r\n        float  w      = lerp(1.0, wGauss, weighting);\r\n\r\n        accum += col * w;\r\n        wsum  += w;\r\n    }\r\n}\r\n\r\nfloat3 finalCol = accum / max(wsum, 1e-4);\r\nreturn float4(finalCol, 1);\r\n"
   description="REDO 3: With parameterized features: subsampling, gaussian weighting, sample area scale"
   showCode=true
   inputs [
    {
     inputName=ScreenPosition
     input {expression=&self:MaterialExpressionDivide_0}
    }
    {
     inputName=ViewportSize
     input {expression=&self:MaterialExpressionViewSize_0}
    }
    {
     inputName=BlockSize
     input {expression=&self:MaterialExpressionLinearInterpolate_8}
    }
    {
     inputName=SceneTexture
     input {
      expression=&self:MaterialExpressionSceneColor_0
      mask=1
      maskB=1
      maskG=1
      maskR=1
     }
    }
    {
     inputName=Threshold
     input {expression=&self:MaterialExpressionScalarParameter_2}
    }
    {
     inputName=SampleAreaScale
     input {expression=&self:MaterialExpressionScalarParameter_20}
    }
    {
     inputName=Subsample
     input {expression=&self:MaterialExpressionScalarParameter_21}
    }
    {
     inputName=SamplesPerAxis
     input {expression=&self:MaterialExpressionScalarParameter_22}
    }
    {
     inputName=Weighting
     input {expression=&self:MaterialExpressionScalarParameter_23}
    }
    {
     inputName=GaussFalloff
     input {expression=&self:MaterialExpressionScalarParameter_24}
    }
   ]
  }
  NamedRerouteDeclaration(None=@n8) {name=Pixelated, nodeColor=<rgba 0.8516 1 0 1>}
  MaterialExpressionSceneTexture_6 = SceneTexture() {sceneTextureId=PPI_BaseColor}
  n9 = Custom!(
   ScreenPosition=@MaterialExpressionDivide_3
   ViewportSize=@MaterialExpressionViewSize_2
   BlockSize=@MaterialExpressionLinearInterpolate_0
   SceneTexture=@MaterialExpressionSceneTexture_6.Color
  ) {
   code="float2 viewportUV = ScreenPosition.xy;\r\n\r\n\r\nfloat2 sceneUV =\r\n    viewportUV *\r\n    View.ViewSizeAndInvSize.xy *\r\n    View.BufferSizeAndInvSize.zw;\r\n\r\nfloat2 bufferSize = View.BufferSizeAndInvSize.xy;\r\n\r\nfloat2 pixelCoord = floor(sceneUV * bufferSize);\r\n\r\nfloat2 blockStart = floor(pixelCoord / BlockSize) * BlockSize;\r\n\r\nfloat3 accum = 0;\r\nint count = 0;\r\n\r\n\r\n\r\n[loop]\r\nfor (int y = 0; y < BlockSize; y++)\r\n{\r\n    [loop]\r\n    for (int x = 0; x < BlockSize; x++)\r\n    {\r\n        \r\n        float2 sampleCoord = blockStart + float2(x, y);\r\n        float2 sampleUV = (sampleCoord + 0.5) / bufferSize;\r\n\r\n\r\n        float3 col = SceneTextureLookup(sampleUV, 5, true).rgb;\r\n\r\n        //float3 col = DecodeSceneColorAndAlpharForMaterialNode(sampleUV).rgb;\r\n\r\n        accum += col;\r\n        count++;\r\n    }\r\n}\r\n\r\nreturn float4(accum / count, 1);"
   showCode=true
   inputs [
    {
     inputName=ScreenPosition
     input {expression=&self:MaterialExpressionDivide_3}
    }
    {
     inputName=ViewportSize
     input {expression=&self:MaterialExpressionViewSize_2}
    }
    {
     inputName=BlockSize
     input {expression=&self:MaterialExpressionLinearInterpolate_0}
    }
    {
     inputName=SceneTexture
     input {
      expression=&self:MaterialExpressionSceneTexture_6
      mask=1
      maskA=1
      maskB=1
      maskG=1
      maskR=1
     }
    }
   ]
  }
  StaticSwitchParameter(True=@n6, False=@n9) {group=Pixelation, parameterName=Scene Color Vs Base Color}
 ]
}
