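# Prototype unlit translucent material for the CraftGlue female character, experimenting with screen-space pixelation and outlines.
# An inverted-hull anime outline passes through cloth wind and character PostFX functions into MF_Pixelation.
# An aspect-corrected grid samples custom depth around cell centers, comparing it with pixel depth to mask edges.
# A custom HLSL Laplacian over scene depth and world normals derives further outline masks, with no material output connection shown.
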Material /Game/Characters/Player/Female/Medium/Bodies/F_MED_CraftGlue/Prototype/Material/M_Pixelation
graph {
 expressions {
  MaterialExpressionAbs_0 {class=MaterialExpressionAbs, inputs [None {from=MaterialExpressionComponentMask_15, output=""}], properties {}}
  MaterialExpressionAbs_1 {class=MaterialExpressionAbs, inputs [None {from=MaterialExpressionComponentMask_17, output=""}], properties {}}
  MaterialExpressionAdd_1 {
   class=MaterialExpressionAdd
   inputs [
    A {from=MaterialExpressionDivide_0}
    B {from=MaterialExpressionDivide_1}
   ]
   properties {desc=Center of The Square}
  }
  MaterialExpressionAdd_10 {
   class=MaterialExpressionAdd
   inputs [
    A {from=MaterialExpressionAdd_1}
    B {from=MaterialExpressionMultiply_22}
   ]
  }
  MaterialExpressionAdd_2 {
   class=MaterialExpressionAdd
   inputs [
    A {from=MaterialExpressionAdd_1}
    B {from=MaterialExpressionMultiply_2}
   ]
  }
  MaterialExpressionAdd_3 {
   class=MaterialExpressionAdd
   inputs [
    A {from=MaterialExpressionAdd_1}
    B {from=MaterialExpressionMultiply_3}
   ]
  }
  MaterialExpressionAdd_4 {
   class=MaterialExpressionAdd
   inputs [
    A {from=MaterialExpressionAdd_1}
    B {from=MaterialExpressionMultiply_4}
   ]
  }
  MaterialExpressionAdd_5 {
   class=MaterialExpressionAdd
   inputs [
    A {from=MaterialExpressionAdd_1}
    B {from=MaterialExpressionMultiply_5}
   ]
  }
  MaterialExpressionAdd_7 {
   class=MaterialExpressionAdd
   inputs [
    A {from=MaterialExpressionAdd_1}
    B {from=MaterialExpressionMultiply_16}
   ]
  }
  MaterialExpressionAdd_8 {
   class=MaterialExpressionAdd
   inputs [
    A {from=MaterialExpressionAdd_1}
    B {from=MaterialExpressionMultiply_17}
   ]
  }
  MaterialExpressionAdd_9 {
   class=MaterialExpressionAdd
   inputs [
    A {from=MaterialExpressionAdd_1}
    B {from=MaterialExpressionMultiply_18}
   ]
  }
  MaterialExpressionComponentMask_0 {class=MaterialExpressionComponentMask, inputs [None {from=MaterialExpressionViewSize_1}], properties {r=true}}
  MaterialExpressionComponentMask_1 {class=MaterialExpressionComponentMask, inputs [None {from=MaterialExpressionViewSize_1}], properties {g=true}}
  MaterialExpressionComponentMask_10 {class=MaterialExpressionComponentMask, inputs [None {from=MaterialExpressionSceneTexture_9, output=Color}], properties {r=true}}
  MaterialExpressionComponentMask_11 {class=MaterialExpressionComponentMask, inputs [None {from=MaterialExpressionSceneTexture_10, output=Color}], properties {r=true}}
  MaterialExpressionComponentMask_14 {class=MaterialExpressionComponentMask, inputs [None {from=MaterialExpressionComponentMask_18}], properties {g=true}}
  MaterialExpressionComponentMask_15 {class=MaterialExpressionComponentMask, inputs [None {from=MaterialExpressionComponentMask_18}], properties {r=true}}
  MaterialExpressionComponentMask_16 {class=MaterialExpressionComponentMask, inputs [None {from=MaterialExpressionComponentMask_18}], properties {b=true}}
  MaterialExpressionComponentMask_17 {class=MaterialExpressionComponentMask, inputs [None {from=MaterialExpressionCustom_1}], properties {a=true}}
  MaterialExpressionComponentMask_18 {class=MaterialExpressionComponentMask, inputs [None {from=MaterialExpressionCustom_1}], properties {b=true, g=true, r=true}}
  MaterialExpressionComponentMask_3 {class=MaterialExpressionComponentMask, inputs [None {from=MaterialExpressionSceneTexture_0, output=Color}], properties {r=true}}
  MaterialExpressionComponentMask_4 {class=MaterialExpressionComponentMask, inputs [None {from=MaterialExpressionSceneTexture_1, output=Color}], properties {r=true}}
  MaterialExpressionComponentMask_5 {class=MaterialExpressionComponentMask, inputs [None {from=MaterialExpressionSceneTexture_2, output=Color}], properties {r=true}}
  MaterialExpressionComponentMask_6 {class=MaterialExpressionComponentMask, inputs [None {from=MaterialExpressionSceneTexture_3, output=Color}], properties {r=true}}
  MaterialExpressionComponentMask_8 {class=MaterialExpressionComponentMask, inputs [None {from=MaterialExpressionSceneTexture_7, output=Color}], properties {r=true}}
  MaterialExpressionComponentMask_9 {class=MaterialExpressionComponentMask, inputs [None {from=MaterialExpressionSceneTexture_8, output=Color}], properties {r=true}}
  MaterialExpressionComposite_0 {class=MaterialExpressionComposite, inputs [], properties {subgraphName=Construct Base Inputs}}
  MaterialExpressionConstant2Vector_0 {class=MaterialExpressionConstant2Vector, properties {g=1, r=1}}
  MaterialExpressionConstant2Vector_1 {class=MaterialExpressionConstant2Vector, properties {g=-1, r=1}}
  MaterialExpressionConstant2Vector_2 {class=MaterialExpressionConstant2Vector, properties {g=1, r=-1}}
  MaterialExpressionConstant2Vector_3 {class=MaterialExpressionConstant2Vector, properties {g=-1, r=-1}}
  MaterialExpressionConstant_1 {class=MaterialExpressionConstant}
  MaterialExpressionConstant_3 {class=MaterialExpressionConstant, properties {r=1}}
  MaterialExpressionCustom_1 {
   class=MaterialExpressionCustom
   inputs [
    a {from=MaterialExpressionSceneTexture_4, output=Color}
    OutlineSize {from=MaterialExpressionScalarParameter_17, output=""}
    Bias {from=MaterialExpressionScalarParameter_18, output=""}
    UV {from=MaterialExpressionDivide_4, output=""}
    alphatest {from=MaterialExpressionScalarParameter_21, output=""}
    shaperatio {from=MaterialExpressionScalarParameter_22, output=""}
   ]
   properties {
    code="float4 bug = a;\r\n\r\n//GetDefaultSceneTextureUV(Parameters, 1);\r\n\r\n\r\nfloat2 KernelUVs = UV* View.ViewSizeAndInvSize.xy *\r\n    View.BufferSizeAndInvSize.zw;\r\n;\r\nfloat2 TexelSize = GetSceneTextureViewSize(1).zw * Bias;\r\nfloat2 PixelUVs;\r\n\r\nfloat KERNEL_SIZE = floor(OutlineSize);\r\nfloat ALPHA = 1;\r\nfloat SHAPE_RATIO = 1;\r\n\r\nfloat2 RotationVector = float2(cos(ALPHA), sin(ALPHA));\r\n\r\nfloat3 LaplacianFilter_Normal = float3(0,0,0);\r\nfloat LaplacianFilter_Depth = 0;\r\nfloat CenterWeight = 0; //KERNEL_SIZE * KERNEL_SIZE - 1;\r\n\r\nfloat HALF_KERNEL_SIZE = floor(KERNEL_SIZE/2);\r\nfloat HALF_KERNEL_SIZE_SQ = KERNEL_SIZE*KERNEL_SIZE/4;\r\n\r\nfor (float y = - HALF_KERNEL_SIZE; y <= HALF_KERNEL_SIZE; y++)\r\n{\r\n    for(float x =- HALF_KERNEL_SIZE; x <= HALF_KERNEL_SIZE; x++)\r\n    {\r\n\r\n        float2 MarkerPoint = float2(dot(RotationVector, float2(x,y)),\r\n                                    dot(RotationVector, float2 (y,-x)));\r\n        if(dot(MarkerPoint, MarkerPoint) > HALF_KERNEL_SIZE_SQ)\r\n        {\r\n            continue;\r\n        }\r\n\r\n           CenterWeight++;\r\n\r\n           PixelUVs = KernelUVs + TexelSize * float2(x,y);\r\n\r\n           LaplacianFilter_Normal -= SceneTextureLookup(PixelUVs, 8, false).rgb;\r\n           LaplacianFilter_Depth -= SceneTextureLookup(PixelUVs, 1, false).r;\r\n    }\r\n\r\n}\r\n\r\nLaplacianFilter_Normal += SceneTextureLookup(PixelUVs, 8, false).rgb * CenterWeight;\r\nLaplacianFilter_Depth += SceneTextureLookup(PixelUVs, 1, false).r * CenterWeight;\r\n\r\nCenterWeight--;\r\n\r\nCenterWeight = 1 / CenterWeight;\r\nLaplacianFilter_Normal *= CenterWeight;\r\nLaplacianFilter_Depth *= CenterWeight;\r\n\r\nreturn float4(LaplacianFilter_Normal, LaplacianFilter_Depth);"
    outputType=CMOT_Float4
    showCode=true
    inputs [
     {
      inputName=a
      input {
       expression=&self:MaterialExpressionSceneTexture_4
       inputName=None
       mask=1
       maskA=1
       maskB=1
       maskG=1
       maskR=1
       outputIndex=0
      }
     }
     {
      inputName=OutlineSize
      input {
       expression=&self:MaterialExpressionScalarParameter_17
       inputName=None
       mask=0
       maskA=0
       maskB=0
       maskG=0
       maskR=0
       outputIndex=0
      }
     }
     {
      inputName=Bias
      input {
       expression=&self:MaterialExpressionScalarParameter_18
       inputName=None
       mask=0
       maskA=0
       maskB=0
       maskG=0
       maskR=0
       outputIndex=0
      }
     }
     {
      inputName=UV
      input {
       expression=&self:MaterialExpressionDivide_4
       inputName=None
       mask=0
       maskA=0
       maskB=0
       maskG=0
       maskR=0
       outputIndex=0
      }
     }
     {
      inputName=alphatest
      input {
       expression=&self:MaterialExpressionScalarParameter_21
       inputName=None
       mask=0
       maskA=0
       maskB=0
       maskG=0
       maskR=0
       outputIndex=0
      }
     }
     {
      inputName=shaperatio
      input {
       expression=&self:MaterialExpressionScalarParameter_22
       inputName=None
       mask=0
       maskA=0
       maskB=0
       maskG=0
       maskR=0
       outputIndex=0
      }
     }
    ]
   }
  }
  MaterialExpressionDivide_0 {
   class=MaterialExpressionDivide
   inputs [
    A {from=MaterialExpressionFloor_0}
    B {from=MaterialExpressionMaterialFunctionCall_8, output=Result}
   ]
  }
  MaterialExpressionDivide_1 {
   class=MaterialExpressionDivide
   inputs [
    A {from=""}
    B {from=MaterialExpressionMaterialFunctionCall_8, output=Result}
   ]
   properties {constA=0.5, desc=Half Axis Length of One Square}
  }
  MaterialExpressionDivide_2 {
   class=MaterialExpressionDivide
   inputs [
    A {from=MaterialExpressionComponentMask_0}
    B {from=MaterialExpressionComponentMask_1}
   ]
  }
  MaterialExpressionDivide_4 {
   class=MaterialExpressionDivide
   inputs [
    A {from=MaterialExpressionFloor_1}
    B {from=MaterialExpressionScalarParameter_19}
   ]
  }
  MaterialExpressionDivide_5 {
   class=MaterialExpressionDivide
   inputs [
    A {from=MaterialExpressionSubtract_5}
    B {from=MaterialExpressionViewSize_0}
   ]
  }
  MaterialExpressionFloor_0 {class=MaterialExpressionFloor, inputs [None {from=MaterialExpressionMultiply_0, output=""}], properties {}}
  MaterialExpressionFloor_1 {class=MaterialExpressionFloor, inputs [None {from=MaterialExpressionMultiply_13, output=""}], properties {}}
  MaterialExpressionFresnel_0 {
   class=MaterialExpressionFresnel
   inputs [
    ExponentIn {from="", output=""}
    BaseReflectFractionIn {from="", output=""}
    Normal {from="", output=""}
   ]
   properties {baseReflectFraction=0, exponent=1}
  }
  MaterialExpressionGetMaterialAttributes_0 {class=MaterialExpressionGetMaterialAttributes, inputs [None {from=MaterialExpressionMaterialFunctionCall_16, output=Result}]}
  MaterialExpressionInverseLinearInterpolate_2 {
   class=MaterialExpressionInverseLinearInterpolate
   inputs [
    A {from="", output=""}
    B {from="", output=""}
    Value {from=MaterialExpressionMax_5, output=""}
   ]
   properties {constA=0.15, constB=0.2}
  }
  MaterialExpressionInverseLinearInterpolate_3 {
   class=MaterialExpressionInverseLinearInterpolate
   inputs [
    A {from="", output=""}
    B {from="", output=""}
    Value {from=MaterialExpressionAbs_1, output=""}
   ]
   properties {constA=2, constB=10}
  }
  MaterialExpressionLinearInterpolate_4 {
   class=MaterialExpressionLinearInterpolate
   inputs [
    A {from=MaterialExpressionConstant_3}
    B {from=""}
    Alpha {from=MaterialExpressionSubtract_1}
   ]
  }
  MaterialExpressionLinearInterpolate_5 {
   class=MaterialExpressionLinearInterpolate
   inputs [
    A {from=MaterialExpressionNamedRerouteUsage_4}
    B {from=""}
    Alpha {from=MaterialExpressionMax_6}
   ]
   properties {constB=0}
  }
  MaterialExpressionMaterialFunctionCall_0 {
   class=MaterialExpressionMaterialFunctionCall
   inputs []
   properties {materialFunction=/Game/Global/Materials/Character_Parent_2023/Anime_Parent_2023/MaterialFunctions/MFC_AnimeOutline_InvHull_2023}
  }
  MaterialExpressionMaterialFunctionCall_1 {
   class=MaterialExpressionMaterialFunctionCall
   inputs [
    Material {from=MaterialExpressionMaterialFunctionCall_0, output=Result}
    FX {from=MaterialExpressionMaterialFunctionCall_0, output=Result}
   ]
   properties {materialFunction=/Game/Packages/Fortress_SharedMaterials/Anime_Master_Material/MFC_ClothWind}
  }
  MaterialExpressionMaterialFunctionCall_14 {
   class=MaterialExpressionMaterialFunctionCall
   inputs [Material {from=""}]
   properties {materialFunction=/Game/Packages/Fortress_SharedMaterials/Base_Material_Functions/CharacterBrightnessandScale_Balance}
  }
  MaterialExpressionMaterialFunctionCall_16 {
   class=MaterialExpressionMaterialFunctionCall
   inputs [
    Material {from=MaterialExpressionMaterialFunctionCall_1, output=Result}
    M {from=MaterialExpressionReroute_8, output=M}
    Pre Skinned Local Position {from=MaterialExpressionReroute_9, output=Pre Skinned Local Position}
    Distance Blend {from=MaterialExpressionReroute_10, output=Distance Blend}
    Fresnel {from=MaterialExpressionReroute_11, output=Fresnel}
   ]
   properties {materialFunction=&MF_CharacterParent_2023_PostFX}
  }
  MaterialExpressionMaterialFunctionCall_3 {
   class=MaterialExpressionMaterialFunctionCall
   inputs [MaterialAttributes {from=MaterialExpressionGetMaterialAttributes_0, output=MaterialAttributes}]
   properties {materialFunction=&/MF_Pixelation}
  }
  MaterialExpressionMaterialFunctionCall_8 {
   class=MaterialExpressionMaterialFunctionCall
   inputs [
    X {from=MaterialExpressionMultiply_6}
    Y {from=MaterialExpressionScalarParameter_5}
   ]
   properties {materialFunction=&MakeFloat2}
  }
  MaterialExpressionMax_0 {
   class=MaterialExpressionMax
   inputs [
    A {from=MaterialExpressionSphereMask_0}
    B {from=MaterialExpressionSphereMask_2}
   ]
  }
  MaterialExpressionMax_1 {
   class=MaterialExpressionMax
   inputs [
    A {from=MaterialExpressionMax_0}
    B {from=MaterialExpressionSphereMask_3}
   ]
  }
  MaterialExpressionMax_2 {
   class=MaterialExpressionMax
   inputs [
    A {from=MaterialExpressionMax_1}
    B {from=MaterialExpressionSphereMask_4}
   ]
  }
  MaterialExpressionMax_4 {
   class=MaterialExpressionMax
   inputs [
    A {from=MaterialExpressionMultiply_28}
    B {from=MaterialExpressionMultiply_15}
   ]
  }
  MaterialExpressionMax_5 {
   class=MaterialExpressionMax
   inputs [
    A {from=MaterialExpressionMax_4}
    B {from=MaterialExpressionMultiply_27}
   ]
  }
  MaterialExpressionMax_6 {
   class=MaterialExpressionMax
   inputs [
    A {from=MaterialExpressionInverseLinearInterpolate_2}
    B {from=MaterialExpressionInverseLinearInterpolate_3}
   ]
  }
  MaterialExpressionMultiply_0 {
   class=MaterialExpressionMultiply
   inputs [
    A {from=MaterialExpressionScreenPosition_0, output=ViewportUV}
    B {from=MaterialExpressionMaterialFunctionCall_8, output=Result}
   ]
   properties {constB=100}
  }
  MaterialExpressionMultiply_1 {
   class=MaterialExpressionMultiply
   inputs [
    A {from=MaterialExpressionDivide_1}
    B {from=MaterialExpressionScalarParameter_8}
   ]
   properties {constB=1.8}
  }
  MaterialExpressionMultiply_13 {
   class=MaterialExpressionMultiply
   inputs [
    A {from=MaterialExpressionScreenPosition_3, output=ViewportUV}
    B {from=MaterialExpressionScalarParameter_19}
   ]
  }
  MaterialExpressionMultiply_14 {
   class=MaterialExpressionMultiply
   inputs [
    A {from=MaterialExpressionScreenPosition_3, output=ViewportUV}
    B {from=MaterialExpressionViewSize_0}
   ]
  }
  MaterialExpressionMultiply_15 {
   class=MaterialExpressionMultiply
   inputs [
    A {from=MaterialExpressionComponentMask_14}
    B {from=MaterialExpressionScalarParameter_23}
   ]
  }
  MaterialExpressionMultiply_16 {
   class=MaterialExpressionMultiply
   inputs [
    A {from=MaterialExpressionMultiply_23}
    B {from=MaterialExpressionConstant2Vector_0}
   ]
  }
  MaterialExpressionMultiply_17 {
   class=MaterialExpressionMultiply
   inputs [
    A {from=MaterialExpressionMultiply_23}
    B {from=MaterialExpressionConstant2Vector_1}
   ]
  }
  MaterialExpressionMultiply_18 {
   class=MaterialExpressionMultiply
   inputs [
    A {from=MaterialExpressionMultiply_23}
    B {from=MaterialExpressionConstant2Vector_2}
   ]
  }
  MaterialExpressionMultiply_19 {
   class=MaterialExpressionMultiply
   inputs [
    A {from=MaterialExpressionSphereMask_0}
    B {from=MaterialExpressionSphereMask_2}
   ]
  }
  MaterialExpressionMultiply_2 {
   class=MaterialExpressionMultiply
   inputs [
    A {from=MaterialExpressionMultiply_1}
    B {from=MaterialExpressionConstant2Vector_0}
   ]
  }
  MaterialExpressionMultiply_20 {
   class=MaterialExpressionMultiply
   inputs [
    A {from=MaterialExpressionMultiply_19}
    B {from=MaterialExpressionSphereMask_3}
   ]
  }
  MaterialExpressionMultiply_21 {
   class=MaterialExpressionMultiply
   inputs [
    A {from=MaterialExpressionMultiply_20}
    B {from=MaterialExpressionSphereMask_4}
   ]
  }
  MaterialExpressionMultiply_22 {
   class=MaterialExpressionMultiply
   inputs [
    A {from=MaterialExpressionMultiply_23}
    B {from=MaterialExpressionConstant2Vector_3}
   ]
  }
  MaterialExpressionMultiply_23 {
   class=MaterialExpressionMultiply
   inputs [
    A {from=MaterialExpressionDivide_1}
    B {from=MaterialExpressionScalarParameter_12}
   ]
  }
  MaterialExpressionMultiply_24 {
   class=MaterialExpressionMultiply
   inputs [
    A {from=MaterialExpressionSphereMask_1}
    B {from=MaterialExpressionSphereMask_5}
   ]
  }
  MaterialExpressionMultiply_25 {
   class=MaterialExpressionMultiply
   inputs [
    A {from=MaterialExpressionMultiply_24}
    B {from=MaterialExpressionSphereMask_6}
   ]
  }
  MaterialExpressionMultiply_26 {
   class=MaterialExpressionMultiply
   inputs [
    A {from=MaterialExpressionMultiply_25}
    B {from=MaterialExpressionSphereMask_7}
   ]
  }
  MaterialExpressionMultiply_27 {
   class=MaterialExpressionMultiply
   inputs [
    A {from=MaterialExpressionComponentMask_16}
    B {from=MaterialExpressionScalarParameter_24}
   ]
  }
  MaterialExpressionMultiply_28 {
   class=MaterialExpressionMultiply
   inputs [
    A {from=MaterialExpressionAbs_0}
    B {from=MaterialExpressionScalarParameter_26}
   ]
  }
  MaterialExpressionMultiply_3 {
   class=MaterialExpressionMultiply
   inputs [
    A {from=MaterialExpressionMultiply_1}
    B {from=MaterialExpressionConstant2Vector_1}
   ]
  }
  MaterialExpressionMultiply_4 {
   class=MaterialExpressionMultiply
   inputs [
    A {from=MaterialExpressionMultiply_1}
    B {from=MaterialExpressionConstant2Vector_2}
   ]
  }
  MaterialExpressionMultiply_5 {
   class=MaterialExpressionMultiply
   inputs [
    A {from=MaterialExpressionMultiply_1}
    B {from=MaterialExpressionConstant2Vector_3}
   ]
  }
  MaterialExpressionMultiply_6 {
   class=MaterialExpressionMultiply
   inputs [
    A {from=MaterialExpressionScalarParameter_5}
    B {from=MaterialExpressionDivide_2}
   ]
  }
  MaterialExpressionNamedRerouteUsage_3 {class=MaterialExpressionNamedRerouteUsage}
  MaterialExpressionNamedRerouteUsage_4 {class=MaterialExpressionNamedRerouteUsage}
  MaterialExpressionPinBase_0 {class=MaterialExpressionPinBase, inputs [], properties {}}
  MaterialExpressionPinBase_1 {
   class=MaterialExpressionPinBase
   inputs [
    M {from=MaterialExpressionConstant_1, output=""}
    Pre Skinned Local Position {from=MaterialExpressionShaderStageSwitch_0, output=""}
    Distance Blend {from=MaterialExpressionMaterialFunctionCall_14, output=Distance Blend}
    Fresnel {from=MaterialExpressionFresnel_0, output=""}
   ]
   properties {reroutePins [
     M {expression=&self:MaterialGraph_1.MaterialGraphNode_Composite_0.Construct Base Inputs.MaterialExpressionReroute_8}
     Pre Skinned Local Position {expression=&self:MaterialGraph_1.MaterialGraphNode_Composite_0.Construct Base Inputs.MaterialExpressionReroute_9}
     Distance Blend {expression=&self:MaterialGraph_1.MaterialGraphNode_Composite_0.Construct Base Inputs.MaterialExpressionReroute_10}
     Fresnel {expression=&self:MaterialGraph_1.MaterialGraphNode_Composite_0.Construct Base Inputs.MaterialExpressionReroute_11}
    ]}
  }
  MaterialExpressionPixelDepth_0 {class=MaterialExpressionPixelDepth, inputs [], properties {}}
  MaterialExpressionPixelDepth_2 {class=MaterialExpressionPixelDepth, inputs [], properties {}}
  MaterialExpressionPreSkinnedPosition_0 {class=MaterialExpressionPreSkinnedPosition, inputs [], properties {}}
  MaterialExpressionSaturate_1 {class=MaterialExpressionSaturate, inputs [None {from=MaterialExpressionSubtract_1}]}
  MaterialExpressionSaturate_4 {class=MaterialExpressionSaturate, inputs [None {from=MaterialExpressionInverseLinearInterpolate_2}]}
  MaterialExpressionScalarParameter_11 {class=MaterialExpressionScalarParameter, properties {defaultValue=12, parameterName=Depth Tolerance}}
  MaterialExpressionScalarParameter_12 {class=MaterialExpressionScalarParameter, properties {defaultValue=5.6164, parameterName=Param_4}}
  MaterialExpressionScalarParameter_15 {class=MaterialExpressionScalarParameter, properties {defaultValue=0.292, parameterName=Param_2}}
  MaterialExpressionScalarParameter_16 {class=MaterialExpressionScalarParameter, properties {defaultValue=100, parameterName=Param_3}}
  MaterialExpressionScalarParameter_17 {class=MaterialExpressionScalarParameter, properties {defaultValue=10.0849, group=Outline, parameterName=Outline Size}}
  MaterialExpressionScalarParameter_18 {class=MaterialExpressionScalarParameter, properties {defaultValue=0.5, group=Outline, parameterName=Texel Size Bias}}
  MaterialExpressionScalarParameter_19 {class=MaterialExpressionScalarParameter, properties {defaultValue=379.9195, parameterName=Param_5}}
  MaterialExpressionScalarParameter_21 {class=MaterialExpressionScalarParameter, properties {defaultValue=1, parameterName=Alpha}}
  MaterialExpressionScalarParameter_22 {class=MaterialExpressionScalarParameter, properties {defaultValue=13.5402, parameterName=Shape Ratio}}
  MaterialExpressionScalarParameter_23 {class=MaterialExpressionScalarParameter, properties {group=Outline, parameterName=Inner Outline 02 - Threshold}}
  MaterialExpressionScalarParameter_24 {class=MaterialExpressionScalarParameter, properties {group=Outline, parameterName=Inner Outline 03 - Threshold}}
  MaterialExpressionScalarParameter_26 {class=MaterialExpressionScalarParameter, properties {defaultValue=0.2, group=Outline, parameterName=Inner Outline 01 - Threshold}}
  MaterialExpressionScalarParameter_3 {class=MaterialExpressionScalarParameter, properties {defaultValue=12, parameterName=Depth Tolerance}}
  MaterialExpressionScalarParameter_5 {class=MaterialExpressionScalarParameter, properties {defaultValue=128, parameterName=Pixel Amount B}}
  MaterialExpressionScalarParameter_8 {class=MaterialExpressionScalarParameter, properties {defaultValue=1.8, parameterName=Kernel Size}}
  MaterialExpressionSceneTexture_0 {
   class=MaterialExpressionSceneTexture
   inputs [UVs {from=MaterialExpressionAdd_2, output=""}]
   properties {bFiltered=true, sceneTextureId=PPI_CustomDepth}
  }
  MaterialExpressionSceneTexture_1 {
   class=MaterialExpressionSceneTexture
   inputs [UVs {from=MaterialExpressionAdd_5, output=""}]
   properties {bFiltered=true, sceneTextureId=PPI_CustomDepth}
  }
  MaterialExpressionSceneTexture_10 {
   class=MaterialExpressionSceneTexture
   inputs [UVs {from=MaterialExpressionAdd_8, output=""}]
   properties {bFiltered=true, sceneTextureId=PPI_CustomDepth}
  }
  MaterialExpressionSceneTexture_2 {
   class=MaterialExpressionSceneTexture
   inputs [UVs {from=MaterialExpressionAdd_4, output=""}]
   properties {bFiltered=true, sceneTextureId=PPI_CustomDepth}
  }
  MaterialExpressionSceneTexture_3 {
   class=MaterialExpressionSceneTexture
   inputs [UVs {from=MaterialExpressionAdd_3, output=""}]
   properties {bFiltered=true, sceneTextureId=PPI_CustomDepth}
  }
  MaterialExpressionSceneTexture_4 {class=MaterialExpressionSceneTexture, inputs [UVs {from="", output=""}], properties {sceneTextureId=PPI_CustomDepth}}
  MaterialExpressionSceneTexture_7 {
   class=MaterialExpressionSceneTexture
   inputs [UVs {from=MaterialExpressionAdd_7, output=""}]
   properties {bFiltered=true, sceneTextureId=PPI_CustomDepth}
  }
  MaterialExpressionSceneTexture_8 {
   class=MaterialExpressionSceneTexture
   inputs [UVs {from=MaterialExpressionAdd_10, output=""}]
   properties {bFiltered=true, sceneTextureId=PPI_CustomDepth}
  }
  MaterialExpressionSceneTexture_9 {
   class=MaterialExpressionSceneTexture
   inputs [UVs {from=MaterialExpressionAdd_9, output=""}]
   properties {bFiltered=true, sceneTextureId=PPI_CustomDepth}
  }
  MaterialExpressionScreenPosition_0 {class=MaterialExpressionScreenPosition, inputs [], properties {}}
  MaterialExpressionScreenPosition_3 {class=MaterialExpressionScreenPosition, inputs [], properties {}}
  MaterialExpressionShaderStageSwitch_0 {
   class=MaterialExpressionShaderStageSwitch
   inputs [
    PixelShader {from=MaterialExpressionVertexInterpolator_0, output=PS}
    VertexShader {from=MaterialExpressionPreSkinnedPosition_0, output=""}
   ]
   properties {}
  }
  MaterialExpressionSphereMask_0 {
   class=MaterialExpressionSphereMask
   inputs [
    A {from=MaterialExpressionPixelDepth_0, output=R}
    B {from=MaterialExpressionComponentMask_3, output=""}
    Radius {from=MaterialExpressionScalarParameter_3, output=""}
    Hardness {from="", output=""}
   ]
   properties {attenuationRadius=4}
  }
  MaterialExpressionSphereMask_1 {
   class=MaterialExpressionSphereMask
   inputs [
    A {from=MaterialExpressionPixelDepth_2, output=R}
    B {from=MaterialExpressionComponentMask_8, output=""}
    Radius {from=MaterialExpressionScalarParameter_11, output=""}
    Hardness {from="", output=""}
   ]
   properties {attenuationRadius=4}
  }
  MaterialExpressionSphereMask_2 {
   class=MaterialExpressionSphereMask
   inputs [
    A {from=MaterialExpressionPixelDepth_0, output=R}
    B {from=MaterialExpressionComponentMask_6, output=""}
    Radius {from=MaterialExpressionScalarParameter_3, output=""}
    Hardness {from="", output=""}
   ]
   properties {attenuationRadius=4}
  }
  MaterialExpressionSphereMask_3 {
   class=MaterialExpressionSphereMask
   inputs [
    A {from=MaterialExpressionPixelDepth_0, output=R}
    B {from=MaterialExpressionComponentMask_5, output=""}
    Radius {from=MaterialExpressionScalarParameter_3, output=""}
    Hardness {from="", output=""}
   ]
   properties {attenuationRadius=4}
  }
  MaterialExpressionSphereMask_4 {
   class=MaterialExpressionSphereMask
   inputs [
    A {from=MaterialExpressionPixelDepth_0, output=R}
    B {from=MaterialExpressionComponentMask_4, output=""}
    Radius {from=MaterialExpressionScalarParameter_3, output=""}
    Hardness {from="", output=""}
   ]
   properties {attenuationRadius=4}
  }
  MaterialExpressionSphereMask_5 {
   class=MaterialExpressionSphereMask
   inputs [
    A {from=MaterialExpressionPixelDepth_2, output=R}
    B {from=MaterialExpressionComponentMask_11, output=""}
    Radius {from=MaterialExpressionScalarParameter_11, output=""}
    Hardness {from="", output=""}
   ]
   properties {attenuationRadius=4}
  }
  MaterialExpressionSphereMask_6 {
   class=MaterialExpressionSphereMask
   inputs [
    A {from=MaterialExpressionPixelDepth_2, output=R}
    B {from=MaterialExpressionComponentMask_10, output=""}
    Radius {from=MaterialExpressionScalarParameter_11, output=""}
    Hardness {from="", output=""}
   ]
   properties {attenuationRadius=4}
  }
  MaterialExpressionSphereMask_7 {
   class=MaterialExpressionSphereMask
   inputs [
    A {from=MaterialExpressionPixelDepth_2, output=R}
    B {from=MaterialExpressionComponentMask_9, output=""}
    Radius {from=MaterialExpressionScalarParameter_11, output=""}
    Hardness {from="", output=""}
   ]
   properties {attenuationRadius=4}
  }
  MaterialExpressionStep_0 {
   class=MaterialExpressionStep
   inputs [
    Y {from=MaterialExpressionScalarParameter_15, output=""}
    X {from=MaterialExpressionSaturate_1, output=""}
   ]
   properties {}
  }
  MaterialExpressionSubtract_1 {
   class=MaterialExpressionSubtract
   inputs [
    A {from=MaterialExpressionMax_2}
    B {from=MaterialExpressionMultiply_21}
   ]
  }
  MaterialExpressionSubtract_4 {
   class=MaterialExpressionSubtract
   inputs [
    A {from=MaterialExpressionNamedRerouteUsage_3}
    B {from=MaterialExpressionSubtract_1}
   ]
  }
  MaterialExpressionSubtract_5 {
   class=MaterialExpressionSubtract
   inputs [
    A {from=MaterialExpressionMultiply_14}
    B {from=MaterialExpressionViewProperty_6, output=Property}
   ]
  }
  MaterialExpressionVertexInterpolator_0 {class=MaterialExpressionVertexInterpolator, inputs [VS {from=MaterialExpressionPreSkinnedPosition_0, output=""}], properties {}}
  MaterialExpressionViewProperty_6 {class=MaterialExpressionViewProperty, inputs [], properties {property=MEVP_TemporalSampleOffset}}
  MaterialExpressionViewSize_0 {class=MaterialExpressionViewSize, inputs [], properties {}}
  MaterialExpressionViewSize_1 {class=MaterialExpressionViewSize, inputs [], properties {}}
 }
 properties {
  allowTranslucentCustomDepthWrites=true
  bAutomaticallySetUsageInEditor=false
  bEnableResponsiveAA=true
  bOutputTranslucentVelocity=true
  bUseMaterialAttributes=true
  bUsedWithClothing=true
  bUsedWithMeshDeformer=true
  bUsedWithMorphTargets=true
  bUsedWithSkeletalMesh=true
  blendMode=BLEND_Translucent
  opacityMaskClipValue=0.01
  pixelDepthOffsetMode=PDOM_Legacy
  shadingModel=MSM_Unlit
  twoSided=true
  usedShadingModels=Unlit
 }
}
